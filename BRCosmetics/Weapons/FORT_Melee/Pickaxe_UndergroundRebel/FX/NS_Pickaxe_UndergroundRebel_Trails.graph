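# Niagara system for the Underground Rebel pickaxe's swing trails, drawn as ribbons by a single BladeTrail emitter.
# It spawns two short-lived anim trails from separate bone or socket pairs and updates them against the skeletal mesh.
# User.Variant sets default and alternate render flags, with the default flag also choosing between two time-of-day color sets and RGB scaling.
# Alpha fades over each particle's life and again by system age, ending the effect quickly.
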NiagaraSystem /BRCosmetics/Weapons/FORT_Melee/Pickaxe_UndergroundRebel/FX/NS_Pickaxe_UndergroundRebel_Trails
dependencies {
 usedDataInterfaces [&NiagaraDataInterfaceColorCurve, &NiagaraDataInterfaceCurve]
 usedDynamicInputs [&MakeIntFromFloat]
 usedModules [
  &self:SystemSpawnScript.SystemScriptSource.SystemScriptGraph.NiagaraNodeAssignment_0.SetVariables_#1
  &self:SystemSpawnScript.SystemScriptSource.SystemScriptGraph.NiagaraNodeAssignment_1.SetVariables_#2
  &SystemState
 ]
}
emitters {BladeTrail {
  emitterUpdateScript [
   &EmitterState()
   &FN_SpawnAnimTrail(Number of Anim Trails=2, Spawn Interval=0)
   /SetVariables(Emitter.RenderEnableDefaultVariant:NiagaraBool=&SetBoolByFloatComparison(A=$User.Variant, B=0.5, Comparison Type=::NewEnumerator6)) {isSetParameters=true}
   /SetVariables(Emitter.RenderEnableAltVariant:NiagaraBool=&SetBoolByFloatComparison(A=$User.Variant, B=0.5)) {isSetParameters=true}
  ]
  particleSpawnScript [
   &FN_InitializeAnimTrailBoneIndecies(First Bone or Socket Index=1, Second Bone or Socket Index=0)
   &FN_InitializeAnimTrailBoneIndecies(AnimTrail Selection=1, First Bone or Socket Index=3, Second Bone or Socket Index=2) {module=FN_InitializeAnimTrailBoneIndecies001}
   &FN_InitializeAnimTrail(
    Color=&MakeCustomLinearColorFromBool(
     Bool=$Emitter.RenderEnableDefaultVariant
     True Color=&FNTimeOfDayColor(
      Night Color=<rgba 0.024 0.026 0.03 1.6>
      Morning Color=<rgba 0.063 0.0504 0.0609 1.5>
      Day Color=<rgba 0.081 0.0825 0.09 1>
      Evening Color=<rgba 0.05 0.0335 0.0349 1.5>
     )
     False Color=&FNTimeOfDayColor(
      Night Color=<rgba 0.024 0.0293 0.04 1>
      Morning Color=<rgba 0.22 0.22 0.22 1>
      Day Color=<rgba 0.23 0.23 0.23 1>
      Evening Color=<rgba 0.22 0.22 0.22 1>
     )
    )
    Lifetime=0.12
    SkeletalMesh=$System.Skeletal Mesh
   )
  ]
  particleUpdateScript [
   &FN_UpdateAnimTrail(Locked Particle RibbonWidth=1, SkeletalMesh=$User.SkeletalMesh) {module=FN_UpdateAnimTrail001}
   &ScaleColor(
    Scale Alpha=&FloatFromCurve(FloatCurve=DataInterface() {propertyValues {Curve {keys [
         {arriveTangent=-1.4225, interpMode=RCIM_Cubic, leaveTangent=-1.4225, tangentMode=RCTM_User, time=0.1, value=1}
         {interpMode=RCIM_Cubic, tangentMode=RCTM_User, time=1, value=0}
        ]}}})
    Scale RGB=&MakeCustomVectorFromBool(
     Bool=$Emitter.RenderEnableDefaultVariant
     True Vector=&MakeVectorFromLinearColorRGB(Color=&ColorFromCurve(ColorCurve=DataInterface() {propertyValues {
         AlphaCurve {keys []}
         BlueCurve {keys [<at 0.0187 1 RCIM_Cubic>,<at 0.8785 0.8>]}
         GreenCurve {keys [<at 0.0187 1 RCIM_Cubic>,<at 0.8785 0.8>]}
         RedCurve {keys [<at 0.0187 1 RCIM_Cubic>,<at 0.8785 0.8>]}
        }}))
     False Vector=<1 1 1>
    )
    ScaleA=true
    ScaleRGB=true
   )
   &ScaleColor(
    Scale Alpha=&FloatFromCurve(
     FloatCurve=DataInterface() {propertyValues {Curve {keys [{interpMode=RCIM_Cubic, tangentMode=RCTM_User, time=0.15, value=1}, {interpMode=RCIM_Cubic, tangentMode=RCTM_User, time=0.27, value=0}]}}}
     CurveIndex=$System.Age
    )
    ScaleA=true
    ScaleRGB=false
   ) {module=ScaleColor001}
  ]
  rendererClasses [&NiagaraRibbonRendererProperties]
  renderers [
   {rendererClass=&NiagaraRibbonRendererProperties}
   {rendererClass=&NiagaraRibbonRendererProperties, rendererIndex=1}
  ]
 }}
summary {
 systemName=NS_Pickaxe_UndergroundRebel_Trails
 emitters [{emitterName=BladeTrail, rendererClasses [&NiagaraRibbonRendererProperties]}]
 userVariables [
  User.bisSwinging {type=NiagaraBool}
  User.NewVector3f {type=Vector3f, description="", defaultValue {struct=Vector3f, value=<0 0 0>}}
  User.RibbonIDOffset {type=NiagaraInt32}
  User.SkeletalMesh {
   type=NiagaraDataInterfaceSkeletalMesh
   defaultValue {
    struct=NiagaraExt_VariableValue_DataInterface
    value {dataInterface=&self:NiagaraDataInterfaceSkeletalMesh_1}
   }
  }
  User.Variant {type=NiagaraFloat, defaultValue=0, description=""}
 ]
}
